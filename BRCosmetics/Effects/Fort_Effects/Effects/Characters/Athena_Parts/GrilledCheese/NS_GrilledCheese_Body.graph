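# Niagara body effect for the GrilledCheese character cosmetic, drawing golden electric arcs and lightning over the skeletal mesh.
# BoneFlow_ArcEnergy wraps partitioned ribbons around bone sets, adding lights and sprites; two sprite emitters add lightning along bones and at a head bone.
# User.FinalStage drives spawn probability, reduced at close distance, where ribbons and body sprites also shrink.
# Hidden-part flags set visibility tags; bisFootHidden also stops head spawning.
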NiagaraSystem /BRCosmetics/Effects/Fort_Effects/Effects/Characters/Athena_Parts/GrilledCheese/NS_GrilledCheese_Body
dependencies {
 usedDataInterfaces [&NiagaraDataInterfaceCurve, &NiagaraDataInterfaceColorCurve]
 usedDynamicInputs [&VectorFromFloat]
 usedModules [&SystemState]
}
emitters {
 BoneFlow_ArcEnergy {
  emitterSpawnScript [
   /SetVariables(Emitter.RandomSeed:NiagaraInt32=/Niagara/DynamicInputs/Random/RandomInt(Randomness Mode:ENiagaraRandomnessMode=::NewEnumerator0, Random Seed:NiagaraInt32=0, Fixed Random Seed:NiagaraBool=false)) {isSetParameters=true, module=SetVariables_#1}
   /Game/Effects/Niagara/Modules/Partition/FN_PartitionAttributes(Number of Bone Sets=5, Particles Per Partition=12)
  ]
  emitterUpdateScript [
   &EmitterState()
   &SpawnBurst_Instantaneous(
    Spawn Count=$Emitter.FN_PartitionAttributes.Spawn Total
    Spawn Probability=&Add_Float(
     A=$User.FinalStage
     B=&Lerp_Float(
      A=-0.65
      B=0
      Alpha=/Game/Effects/Niagara/DynamicInputs/Math/FN_Fit(
       Clamp:NiagaraBool=true
       Value:NiagaraFloat=$Emitter.Scalability.DistanceFraction
       SourceMin:NiagaraFloat=0
       SourceMax:NiagaraFloat=&Divide_Float(A=150, B=2250)
       DestMin:NiagaraFloat=0
       DestMax:NiagaraFloat=1
      )
     )
    )
    Use Spawn Probability=true
   )
  ]
  particleSpawnScript [
   /Game/Effects/Niagara/Modules/Partition/FN_Particle_partition_v5/FN_Particle_Partition_v5(Enable Ribbon ID=true, Pattern:ENiagaraPartitionWaveForm=::NewEnumerator1, Seed:NiagaraInt32=$Emitter.RandomSeed, Shuffle Partition ID=true)
   /SetVariables(
    Particles.BoneComparison:NiagaraFloat=&MakeCustomFloatFromBool(
     Bool=&SetBoolByIntComparison(A=$Particles.Partition.ID, B=2, Comparison Type=::NewEnumerator6)
     True Float=1
     False Float=0
    )
    Particles.ChestComparison:NiagaraFloat=&MakeCustomFloatFromBool(
     Bool=&SetBoolByIntComparison(A=$Particles.Partition.ID, B=4, Comparison Type=::NewEnumerator4)
     True Float=1
     False Float=0
    )
   ) {isSetParameters=true}
   /Game/Effects/Niagara/Modules/BoneFlow/FN_Initialize_BoneFlow_Attributes_v2(
    Angle:NiagaraFloat=&Add_Float(
     A=&Multiply_Float(
      A=&RandomRangeFloat(Minimum=-0.25, Maximum=0.25)
      B=&Multiply_Float(A=$Particles.Partition.NormalizedID, B=10)
     )
     B=&RandomRangeFloat(Minimum=0.4, Maximum=0.3)
    )
    Radius:NiagaraFloat=&Multiply_Float(
     A=&Lerp_Float(
      A=&Lerp_Float(
       A=&RandomRangeFloat(Minimum=11, Maximum=12)
       B=&RandomRangeFloat(Minimum=7, Maximum=9)
       Alpha=$Particles.BoneComparison
      )
      B=&RandomRangeFloat(Minimum=8, Maximum=10)
      Alpha=$Particles.ChestComparison
     )
     B=&FloatFromCurve(FloatCurve=DataInterface() {propertyValues {Curve {keys [<at 0 1>,<at 1 0.1>]}}}, CurveIndex=$Particles.Partition.NormalizedID)
    )
   )
   &InitializeRibbon(
    Color=<rgba 1.25 1.0305 0.3721 1>
    Lifetime=/Game/Effects/Niagara/DynamicInputs/Random/FN_SeededRandomRangeFloat(Minimum:NiagaraFloat=0.175, Maximum:NiagaraFloat=0.35, Random Seed:NiagaraInt32=$Particles.Partition.ID)
    Module.Write Color=true
    Module.Write Lifetime=true
    Module.Write Mass=false
    Module.Write Material Random=true
    Module.Write Position=false
    Module.Write RibbonTwist=false
    Module.Write RibbonWidth=true
    Ribbon Width=17
   )
  ]
  particleUpdateScript [
   &ParticleState()
   &ScaleColor(
    Scale Mode=::NewEnumerator1
    Scale RGBA=&Vector4FromCurve(Vector4Curve=DataInterface() {propertyValues {
       AlphaCurve {keys [<at 0.706 1>,<at 1 0>]}
       BlueCurve {keys [<at 0 50>,<at 0.1818 1>,<at 0.4428 1>,<at 0.6162 10>]}
       GreenCurve {keys [<at 0 50>,<at 0.1818 1>,<at 0.4428 1>,<at 0.6162 10>]}
       RedCurve {keys [<at 0 50>,<at 0.1818 1>,<at 0.4428 1>,<at 0.6162 10>]}
      }})
    ScaleRGBA=true
   )
   &DynamicMaterialParameters(
    Index 0 Param 1=1
    Index 0 Param 2=&FloatFromCurve(FloatCurve=DataInterface() {propertyValues {Curve {keys [
         {arriveTangent=16.9415, interpMode=RCIM_Cubic, leaveTangent=16.9415, tangentMode=RCTM_User, time=0, value=-1}
         {arriveTangent=3.3333, leaveTangent=3.3333, time=0.25, value=1}
         {arriveTangent=0.7567, interpMode=RCIM_Cubic, leaveTangent=0.7567, tangentMode=RCTM_User, time=0.45, value=1}
         <at 1 -1 RCIM_Cubic>
        ]}}})
    Index 0 Param 3=1
    Index 0 Param 4=&RandomRangeFloat(Minimum=0, Maximum=1)
    Index 1 Param 1=&FloatFromCurve(FloatCurve=DataInterface() {propertyValues {Curve {keys [<at 0.4 2 RCIM_Cubic>,<at 1 5>]}}})
    Module.Param0WriteEnabled=false
    Module.Param0WriteEnabled1=true
    Module.Param1WriteEnabled=true
    Module.Param1WriteEnabled1=true
    Module.Param2WriteEnabled=true
    Module.Param2WriteEnabled1=true
    Module.Param3WriteEnabled=true
    Module.Param3WriteEnabled1=true
    Write Parameter Index 1=true
   )
   &FN_BoneFlow(
    Angle:NiagaraFloat=&Multiply_Float(
     A=&RandomRangeFloat(Minimum=-1, Maximum=1)
     B=&Add_Float(
      A=$Particles.Initial.FN_Initialize_BoneFlow_Attributes_v2.Angle
      B=&FloatFromCurve(
       FloatCurve=DataInterface() {propertyValues {Curve {keys [{arriveTangent=6.1308, interpMode=RCIM_Cubic, leaveTangent=6.1308, tangentMode=RCTM_User, time=0, value=0}, <at 1 2 RCIM_Cubic>]}}}
       Scale Curve=0.02
      )
     )
    )
    BoneID:NiagaraInt32=$Particles.Partition.ID
    Height:NiagaraFloat=$Particles.Partition.NormalizedID
    Interpolate Mesh Sampling=false
    Radius:NiagaraFloat=&Add_Float(
     A=$Particles.Initial.FN_Initialize_BoneFlow_Attributes_v2.Radius
     B=&RandomRangeFloat(Minimum=-1, Maximum=1)
    )
    Set Material Dynamic=false
    SkeletalMesh:NiagaraDataInterfaceSkeletalMesh=$User.SkeletalMesh
   )
   &FN_Light_Attributes(
    Color=<rgba 1.5 1.3027 0.7109 1>
    Create Color Attribute=true
    Create Exponent Attribute=true
    Create Position Attribute=true
    Create Radius Attribute=true
    Enable Particle Light Enable=true
    Exponent=2
    Light Enabled=HlslExpression() {hlslExpression="Modulo(Particles.UniqueID, 10) == 0 ? 1: 0"}
    Position=&Add_Vector(
     A=$Output.FN_BoneFlow.PositionOnSpline
     B=&Multiply_VectorByFloat(Vector=&NormalizeVector(Vector To Normalize=$Output.FN_BoneFlow.RotatedBiNormal), Float=10)
    )
    Radius=16
    Scale Color=&FloatFromCurve(FloatCurve=DataInterface() {propertyValues {Curve {keys [
         <at 0 4 RCIM_Cubic>
         {arriveTangent=-3.8319, interpMode=RCIM_Cubic, leaveTangent=-3.8319, time=0.1693, value=1.2561}
         {arriveTangent=0.4201, interpMode=RCIM_Cubic, leaveTangent=0.4201, time=0.4182, value=2.3973}
         {arriveTangent=-4.1208, interpMode=RCIM_Cubic, leaveTangent=-4.1208, time=0.75, value=1.5}
         <at 1 0>
        ]}}})
   )
   /SetVariables(
    Particles.DecalCameraOffset=5
    Particles.DecalColor:LinearColor=&MultiplyLinearColor(
     Linear Color=<rgba 1.25 0.9733 0.3189 0.65>
     Scale Factor=&MakeFloat4FromVectorAndFloat(
      "Vector (RGB)"=&MakeVectorFromLinearColorRGB(Color=&ColorFromCurve(ColorCurve=DataInterface() {propertyValues {
          AlphaCurve {keys []}
          BlueCurve {keys [<at 0 1>, {arriveTangent=-0.1823, interpMode=RCIM_Cubic, leaveTangent=-0.1823, time=0.15, value=2.1354}, <at 1 0.8177 RCIM_Cubic>]}
          GreenCurve {keys [<at 0 1>, {arriveTangent=-0.0456, interpMode=RCIM_Cubic, leaveTangent=-0.0456, time=0.15, value=2.5}, <at 1 0.9544 RCIM_Cubic>]}
          RedCurve {keys [<at 0 1>,<at 0.15 2.4878 RCIM_Cubic>,<at 1 1 RCIM_Cubic>]}
         }}))
      "Float (Alpha)"=&FloatFromCurve(FloatCurve=DataInterface() {propertyValues {Curve {keys [<at 0 0 RCIM_Cubic>,<at 0.2 0.75 RCIM_Cubic>,<at 1 0 RCIM_Cubic>]}}})
     )
    )
    Particles.DecalSize:Vector2f=&Vector2DFromFloat(Value=&Multiply_Float(
      A=19
      B=&MakeCustomFloatFromBool(
       Bool=&SetBoolByFloatComparison(
        A=&ModuloFloat(
         A=&MakeFloatFromInt(Int=$Particles.UniqueID)
         B=&MakeFloatFromInt(Int=&RandomRangeInt(Minimum=4, Maximum=6))
        )
        B=1
        Comparison Type=::NewEnumerator4
       )
       True Float=1
       False Float=0
      )
     ))
   ) {isSetParameters=true}
   &ScaleRibbonWidth(Ribbon Width Scale=&Lerp_Float(
     A=0.5
     B=1.5
     Alpha=&FN_Fit(
      Clamp:NiagaraBool=true
      Value:NiagaraFloat=$Emitter.Scalability.DistanceFraction
      SourceMin:NiagaraFloat=0
      SourceMax:NiagaraFloat=&Divide_Float(A=150, B=2000)
      DestMin:NiagaraFloat=0
      DestMax:NiagaraFloat=1
     )
    ))
   &ScaleRibbonWidth(Ribbon Width Scale=&OneMinusFloat(Float=&MakeFloatFromInt(Int=&self:PartHiddenPartitionCheck(
       is Hand Hidden:NiagaraBool=$User.bisHandHidden
       Selection:NiagaraInt32=$Particles.Partition.ID
       is Foot Hidden:NiagaraBool=$User.bisFootHidden
      )))) {module=ScaleRibbonWidth001}
   /SetVariables(Particles.VisibilityTag=&self:PartHiddenPartitionCheck(
     is Hand Hidden:NiagaraBool=$User.bisHandHidden
     Selection:NiagaraInt32=$Particles.Partition.ID
     is Foot Hidden:NiagaraBool=$User.bisFootHidden
    )) {isSetParameters=true, module=SetVariables_#2}
  ]
  rendererClasses [
   &NiagaraRibbonRendererProperties
   &NiagaraLightRendererProperties
   &NiagaraSpriteRendererProperties
  ]
  renderers [
   {rendererClass=&NiagaraRibbonRendererProperties}
   {rendererClass=&NiagaraLightRendererProperties, rendererIndex=1}
   {rendererClass=&NiagaraSpriteRendererProperties, rendererIndex=2}
  ]
 }
 Lightning_Sprites {
  emitterSpawnScript [
   /SetVariables(Emitter.RandomSeed:NiagaraInt32=&RandomInt(Randomness Mode:ENiagaraRandomnessMode=::NewEnumerator0, Random Seed:NiagaraInt32=0, Fixed Random Seed:NiagaraBool=false)) {isSetParameters=true, module=SetVariables_#1}
   &FN_PartitionAttributes(Number of Bone Sets=6, Particles Per Partition=6, Partitions=6)
  ]
  emitterUpdateScript [
   &EmitterState()
   &SpawnRate(
    Spawn Probability=&Add_Float(
     A=$User.FinalStage
     B=&Lerp_Float(
      A=-0.65
      B=0
      Alpha=&FN_Fit(
       Clamp:NiagaraBool=true
       Value:NiagaraFloat=$Emitter.Scalability.DistanceFraction
       SourceMin:NiagaraFloat=0
       SourceMax:NiagaraFloat=&Divide_Float(A=150, B=2250)
       DestMin:NiagaraFloat=0
       DestMax:NiagaraFloat=1
      )
     )
    )
    SpawnRate=&RandomRangeFloat(Minimum=20, Maximum=40)
    Use Spawn Probability=true
   )
  ]
  particleSpawnScript [
   &FN_Particle_Partition_v5(Pattern:ENiagaraPartitionWaveForm=::NewEnumerator1, Seed:NiagaraInt32=$Emitter.RandomSeed, Shuffle Partition ID=true)
   &FN_Initialize_BoneFlow_Attributes_v2(
    Angle=HlslExpression() {hlslExpression="Particles.Partition.ID == 5 ? rand(Particles.UniqueID) * 6.24 : 3.14 /2"}
    Radius:NiagaraFloat=&Multiply_Float(
     A=HlslExpression() {hlslExpression="Particles.Partition.ID == 5 ? 6 : 1"}
     B=/Niagara/DynamicInputs/Math/LerpMultipleFloats(
      Float Count:ENiagara_MultipleLerpCount=::NewEnumerator0
      1:NiagaraFloat=0.1
      2:NiagaraFloat=2
      3:NiagaraFloat=0.1
      4:NiagaraFloat=?
      5:NiagaraFloat=?
      Alpha:NiagaraFloat=$Particles.Partition.NormalizedID
     )
    )
   )
   /Niagara/Modules/Spawn/Initialization/InitializeParticle(
    Color=<rgba 0.7448 0.6468 0.353 1>
    Lifetime=&UniformRangedFloat(Minimum=0.15, Maximum=0.3)
    Mesh Scale=<3 3 3>
    Module.Write Color=true
    Module.Write Lifetime=true
    Module.Write Mass=false
    Module.Write Material Random=true
    Module.Write Position=false
    Module.Write Scale=true
    Module.Write SpriteRotation=true
    Module.Write SpriteSize=true
    Sprite Rotation=&NormalizedAngleToDegrees(NormalizedAngle=&UniformRangedFloat(Minimum=0, Maximum=1))
    Sprite Size=&Vector2DFromFloat(Value=&UniformRangedFloat(Minimum=14, Maximum=18))
   )
   /SetVariables(Particles.SubImageIndex=&UniformRangedFloat(Minimum=0, Maximum=3)) {isSetParameters=true, module=SetVariables_#3}
   &CameraOffset(Camera Offset Amount=&Lerp_Float(
     A=&RandomRangeFloat(Minimum=-4, Maximum=4)
     B=&RandomRangeFloat(Minimum=-4, Maximum=8)
     Alpha=HlslExpression() {hlslExpression="Particles.Partition.ID >  1 ? 1 : 0"}
    ))
  ]
  particleUpdateScript [
   &ParticleState()
   &DynamicMaterialParameters(
    Index 0 Param 1=&RandomRangeFloat(Minimum=0, Maximum=1)
    Index 0 Param 2=&RandomRangeFloat(Minimum=0, Maximum=0.1)
    Index 0 Param 3=&FloatFromCurve(FloatCurve=DataInterface() {propertyValues {Curve {keys [<at 0 1 RCIM_Cubic>,<at 0.2 0.025 RCIM_Cubic>,<at 1 1>]}}})
    Module.Param0WriteEnabled=true
    Module.Param1WriteEnabled=true
    Module.Param2WriteEnabled=true
    Module.Param3WriteEnabled=false
   )
   &ScaleColor(
    Scale Alpha=&FloatFromCurve(FloatCurve=DataInterface() {propertyValues {Curve {keys [<at 0.75 1 RCIM_Cubic>, {arriveTangent=-10.3917, interpMode=RCIM_Cubic, leaveTangent=-10.3917, tangentMode=RCTM_User, time=1, value=0}]}}})
    Scale RGB=&Multiply_VectorByFloat(
     Vector=&MakeVectorFromLinearColorRGB(Color=&ColorFromCurve(ColorCurve=DataInterface() {propertyValues {
         AlphaCurve {keys []}
         BlueCurve {keys [<at 0 1>, {arriveTangent=-0.1823, interpMode=RCIM_Cubic, leaveTangent=-0.1823, time=0.15, value=1.7083}, <at 1 0.8177 RCIM_Cubic>]}
         GreenCurve {keys [<at 0 1>, {arriveTangent=-0.0091, interpMode=RCIM_Cubic, leaveTangent=-0.0091, time=0.15, value=2}, <at 1 0.9909 RCIM_Cubic>]}
         RedCurve {keys [<at 0 1>,<at 0.15 1.9903 RCIM_Cubic>,<at 1 1 RCIM_Cubic>]}
        }}))
     Float=1
    )
    ScaleA=true
    ScaleRGB=true
   )
   &FN_BoneFlow(
    Angle:NiagaraFloat=$Particles.FN_Initialize_BoneFlow_Attributes_v2.Angle
    Height:NiagaraFloat=$Particles.Partition.NormalizedID
    Interpolate Mesh Sampling=false
    Radius:NiagaraFloat=$Particles.FN_Initialize_BoneFlow_Attributes_v2.Radius
    SkeletalMesh:NiagaraDataInterfaceSkeletalMesh=$User.SkeletalMesh
    Write Position Attribute=true
   )
   &ScaleSpriteSize(Scale Factor=&Vector2DFromFloat(Value=&Lerp_Float(
      A=0.85
      B=1
      Alpha=&FN_Fit(
       Clamp:NiagaraBool=true
       Value:NiagaraFloat=$Emitter.Scalability.DistanceFraction
       SourceMin:NiagaraFloat=0
       SourceMax:NiagaraFloat=&Divide_Float(A=150, B=2000)
       DestMin:NiagaraFloat=0
       DestMax:NiagaraFloat=1
      )
     )))
   /SetVariables(Particles.VisibilityTag=&self:PartHiddenPartitionCheck(
     is Hand Hidden:NiagaraBool=$User.bisHandHidden
     Selection:NiagaraInt32=$Particles.Partition.ID
     is Foot Hidden:NiagaraBool=$User.bisFootHidden
    )) {isSetParameters=true, module=SetVariables_#2}
  ]
  rendererClasses [&NiagaraSpriteRendererProperties]
  renderers [{rendererClass=&NiagaraSpriteRendererProperties}]
 }
 Lightning_Sprites_Head {
  emitterUpdateScript [
   &EmitterState()
   &SpawnRate(
    Spawn Probability=&Add_Float(
     A=$User.FinalStage
     B=&Lerp_Float(
      A=-0.65
      B=0
      Alpha=&FN_Fit(
       Clamp:NiagaraBool=true
       Value:NiagaraFloat=$Emitter.Scalability.DistanceFraction
       SourceMin:NiagaraFloat=0
       SourceMax:NiagaraFloat=&Divide_Float(A=150, B=2250)
       DestMin:NiagaraFloat=0
       DestMax:NiagaraFloat=1
      )
     )
    )
    SpawnRate=&MakeCustomFloatFromBool(Bool=$User.bisFootHidden, True Float=0, False Float=&RandomRangeFloat(Minimum=10, Maximum=16))
    Use Spawn Probability=true
   )
  ]
  particleSpawnScript [
   &FN_Particle_Partition_v5(Pattern:ENiagaraPartitionWaveForm=::NewEnumerator1, Seed:NiagaraInt32=$Emitter.RandomSeed, Shuffle Partition ID=true)
   &InitializeParticle(
    Color=<rgba 0.7448 0.6468 0.353 1>
    Lifetime=&UniformRangedFloat(Minimum=0.15, Maximum=0.3)
    Mesh Scale=<3 3 3>
    Module.Write Color=true
    Module.Write Lifetime=true
    Module.Write Mass=false
    Module.Write Material Random=true
    Module.Write Position=false
    Module.Write Scale=true
    Module.Write SpriteRotation=true
    Module.Write SpriteSize=true
    Sprite Rotation=&NormalizedAngleToDegrees(NormalizedAngle=&UniformRangedFloat(Minimum=0, Maximum=1))
    Sprite Size=&Vector2DFromFloat(Value=&UniformRangedFloat(Minimum=10, Maximum=14))
   )
   /SetVariables(Particles.SubImageIndex=&UniformRangedFloat(Minimum=0, Maximum=3)) {isSetParameters=true, module=SetVariables_#3}
   &CameraOffset(Camera Offset Amount=&RandomRangeFloat(Minimum=-4, Maximum=4))
  ]
  particleUpdateScript [
   &ParticleState()
   &DynamicMaterialParameters(
    Index 0 Param 1=&RandomRangeFloat(Minimum=0, Maximum=1)
    Index 0 Param 2=&RandomRangeFloat(Minimum=0, Maximum=0.1)
    Index 0 Param 3=&FloatFromCurve(FloatCurve=DataInterface() {propertyValues {Curve {keys [<at 0 1 RCIM_Cubic>,<at 0.2 0.025 RCIM_Cubic>,<at 1 1>]}}})
    Module.Param0WriteEnabled=true
    Module.Param1WriteEnabled=true
    Module.Param2WriteEnabled=true
    Module.Param3WriteEnabled=false
   )
   &ScaleColor(
    Scale Alpha=&FloatFromCurve(FloatCurve=DataInterface() {propertyValues {Curve {keys [<at 0.75 1 RCIM_Cubic>, {arriveTangent=-10.3917, interpMode=RCIM_Cubic, leaveTangent=-10.3917, tangentMode=RCTM_User, time=1, value=0}]}}})
    Scale RGB=&Multiply_VectorByFloat(
     Vector=&MakeVectorFromLinearColorRGB(Color=&ColorFromCurve(ColorCurve=DataInterface() {propertyValues {
         AlphaCurve {keys []}
         BlueCurve {keys [<at 0 1>, {arriveTangent=-0.1823, interpMode=RCIM_Cubic, leaveTangent=-0.1823, time=0.15, value=1.7083}, <at 1 0.8177 RCIM_Cubic>]}
         GreenCurve {keys [<at 0 1>, {arriveTangent=-0.0091, interpMode=RCIM_Cubic, leaveTangent=-0.0091, time=0.15, value=2}, <at 1 0.9909 RCIM_Cubic>]}
         RedCurve {keys [<at 0 1>,<at 0.15 1.9903 RCIM_Cubic>,<at 1 1 RCIM_Cubic>]}
        }}))
     Float=1
    )
    ScaleA=true
    ScaleRGB=true
   )
   &SkeletalMeshLocation(
    Bone / Socket Index=18
    Bone Sampling Mode=::NewEnumerator4
    Inherit Velocity=false
    MaskSpawnGroup=false
    Sampled Position Offset=&RandomRangeVector(Minimum=<12 -4 -2>, Maximum=<20 4 2>)
    Sampled Position Offset Coordinate Space:ENiagaraMeshTransforms=::NewEnumerator3
    Skeletal Mesh=$User.SkeletalMesh
   )
   /SetVariables(Particles.VisibilityTag=&MakeCustomIntFromBool(Bool=$User.bisHeadHidden, True INT=1, False INT=0)) {isSetParameters=true}
  ]
  rendererClasses [&NiagaraSpriteRendererProperties]
  renderers [{rendererClass=&NiagaraSpriteRendererProperties}]
 }
}
summary {
 systemName=NS_GrilledCheese_Body
 emitters [
  {
   emitterName=BoneFlow_ArcEnergy
   rendererClasses [
    &NiagaraRibbonRendererProperties
    &NiagaraLightRendererProperties
    &NiagaraSpriteRendererProperties
   ]
  }
  {emitterName=Lightning_Sprites, rendererClasses [&NiagaraSpriteRendererProperties]}
  {emitterName=Lightning_Sprites_Head, rendererClasses [&NiagaraSpriteRendererProperties]}
 ]
 userVariables [
  User.bisFootHidden {type=NiagaraBool}
  User.bisHandHidden {type=NiagaraBool}
  User.bisHeadHidden {type=NiagaraBool}
  User.FinalStage {type=NiagaraFloat, defaultValue=0, description=""}
  User.PlayerSpeed {type=NiagaraFloat, defaultValue=0, description=""}
  User.RibbonIDOffset {type=NiagaraInt32}
  User.SkeletalMesh {
   type=NiagaraDataInterfaceSkeletalMesh
   defaultValue {
    struct=NiagaraExt_VariableValue_DataInterface
    value {dataInterface=&self:NiagaraDataInterfaceSkeletalMesh_0}
   }
  }
  User.TheColor {type=LinearColor, description="", defaultValue {struct=LinearColor, value=<rgba 1 1 1 1>}}
 ]
}
