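# Niagara burst effect for the GrilledCheese character cosmetic that wraps the body in warm yellow-white electric energy.
# Arc ribbons with attached lights and lightning sprites follow bone chains via partitioned BoneFlow, with extra lightning pinned to the head.
# Decal sprites and gravity-driven sparks spawn on sampled bones of a second skeletal mesh input, with age-based buildup curves.
# User bools hide hand, foot and head particles through visibility tags or ribbon width, and IsFrontEnd shortens the decal.
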NiagaraSystem /BRCosmetics/Effects/Fort_Effects/Effects/Characters/Athena_Parts/GrilledCheese/NS_GrilledCheese_Burst
dependencies {
 usedDataInterfaces [&NiagaraDataInterfaceCurve, &NiagaraDataInterfaceColorCurve]
 usedDynamicInputs [&VectorFromFloat]
 usedModules [&SystemState]
}
emitters {
 BoneFlow_ArcEnergy {
  emitterSpawnScript [
   /SetVariables(Emitter.RandomSeed:NiagaraInt32=/Niagara/DynamicInputs/Random/RandomInt(Randomness Mode:ENiagaraRandomnessMode=::NewEnumerator0, Random Seed:NiagaraInt32=0, Fixed Random Seed:NiagaraBool=false)) {isSetParameters=true, module=SetVariables_#1}
   /Game/Effects/Niagara/Modules/Partition/FN_PartitionAttributes(Number of Bone Sets=5, Particles Per Partition=12, Partitions=5)
  ]
  emitterUpdateScript [
   &EmitterState()
   &SpawnBurst_Instantaneous(Spawn Count=$Emitter.FN_PartitionAttributes.Spawn Total, Use Spawn Probability=true)
  ]
  particleSpawnScript [
   /Game/Effects/Niagara/Modules/Partition/FN_Particle_partition_v5/FN_Particle_Partition_v5(Enable Ribbon ID=true, Pattern:ENiagaraPartitionWaveForm=::NewEnumerator1, Seed:NiagaraInt32=$Emitter.RandomSeed)
   /SetVariables(
    Particles.BoneComparison:NiagaraFloat=&MakeCustomFloatFromBool(
     Bool=&SetBoolByIntComparison(A=$Particles.Partition.ID, B=2, Comparison Type=::NewEnumerator6)
     True Float=1
     False Float=0
    )
    Particles.ChestComparison:NiagaraFloat=&MakeCustomFloatFromBool(
     Bool=&SetBoolByIntComparison(A=$Particles.Partition.ID, B=4, Comparison Type=::NewEnumerator4)
     True Float=1
     False Float=0
    )
   ) {isSetParameters=true}
   /Game/Effects/Niagara/Modules/BoneFlow/FN_Initialize_BoneFlow_Attributes_v2(
    Angle:NiagaraFloat=&Add_Float(
     A=&Multiply_Float(
      A=&RandomRangeFloat(Minimum=-0.25, Maximum=0.25)
      B=&Multiply_Float(A=$Particles.Partition.NormalizedID, B=10)
     )
     B=&RandomRangeFloat(Minimum=0.4, Maximum=0.3)
    )
    Radius:NiagaraFloat=&Multiply_Float(
     A=&Lerp_Float(
      A=&Lerp_Float(
       A=&RandomRangeFloat(Minimum=14, Maximum=18)
       B=&RandomRangeFloat(Minimum=11, Maximum=12)
       Alpha=$Particles.BoneComparison
      )
      B=&RandomRangeFloat(Minimum=14, Maximum=16)
      Alpha=$Particles.ChestComparison
     )
     B=&FloatFromCurve(FloatCurve=DataInterface() {propertyValues {Curve {keys [<at 0 1>,<at 1 0.1>]}}}, CurveIndex=$Particles.Partition.NormalizedID)
    )
   )
   &InitializeRibbon(
    Color=<rgba 1.25 1.0305 0.3721 1>
    Lifetime=/Game/Effects/Niagara/DynamicInputs/Random/FN_SeededRandomRangeFloat(Minimum:NiagaraFloat=0.25, Maximum:NiagaraFloat=0.45, Random Seed:NiagaraInt32=$Particles.Partition.ID)
    Module.Write Color=true
    Module.Write Lifetime=true
    Module.Write Mass=false
    Module.Write Material Random=true
    Module.Write Position=false
    Module.Write RibbonTwist=false
    Module.Write RibbonWidth=true
    Ribbon Width=10
   )
  ]
  particleUpdateScript [
   &ParticleState()
   &ScaleColor(
    Scale Mode=::NewEnumerator1
    Scale RGBA=&Vector4FromCurve(Vector4Curve=DataInterface() {propertyValues {
       AlphaCurve {keys [<at 0.706 1>,<at 1 0>]}
       BlueCurve {keys [<at 0 50>,<at 0.1818 1>,<at 0.4428 1>,<at 0.6162 10>]}
       GreenCurve {keys [<at 0 50>,<at 0.1818 1>,<at 0.4428 1>,<at 0.6162 10>]}
       RedCurve {keys [<at 0 50>,<at 0.1818 1>,<at 0.4428 1>,<at 0.6162 10>]}
      }})
    ScaleRGBA=true
   )
   &DynamicMaterialParameters(
    Index 0 Param 1=1
    Index 0 Param 2=&FloatFromCurve(FloatCurve=DataInterface() {propertyValues {Curve {keys [
         {arriveTangent=16.9415, interpMode=RCIM_Cubic, leaveTangent=16.9415, tangentMode=RCTM_User, time=0, value=-1}
         {arriveTangent=3.3333, leaveTangent=3.3333, time=0.25, value=1}
         {arriveTangent=0.7567, interpMode=RCIM_Cubic, leaveTangent=0.7567, tangentMode=RCTM_User, time=0.45, value=1}
         <at 1 -1 RCIM_Cubic>
        ]}}})
    Index 0 Param 3=1
    Index 0 Param 4=&RandomRangeFloat(Minimum=0, Maximum=1)
    Module.Param0WriteEnabled=false
    Module.Param1WriteEnabled=true
    Module.Param2WriteEnabled=true
    Module.Param3WriteEnabled=true
   )
   &FN_BoneFlow(
    Angle:NiagaraFloat=&Multiply_Float(
     A=&RandomRangeFloat(Minimum=-1, Maximum=1)
     B=&Add_Float(
      A=$Particles.Initial.FN_Initialize_BoneFlow_Attributes_v2.Angle
      B=&FloatFromCurve(
       FloatCurve=DataInterface() {propertyValues {Curve {keys [{arriveTangent=6.1308, interpMode=RCIM_Cubic, leaveTangent=6.1308, tangentMode=RCTM_User, time=0, value=0}, <at 1 2 RCIM_Cubic>]}}}
       Scale Curve=0.02
      )
     )
    )
    Height:NiagaraFloat=$Particles.Partition.NormalizedID
    Interpolate Mesh Sampling=false
    Radius:NiagaraFloat=&Add_Float(
     A=$Particles.Initial.FN_Initialize_BoneFlow_Attributes_v2.Radius
     B=&RandomRangeFloat(Minimum=-1, Maximum=1)
    )
    SkeletalMesh:NiagaraDataInterfaceSkeletalMesh=$User.SkeletalMesh
    Write Position Attribute=true
   )
   &FN_Light_Attributes(
    Color=<rgba 1.5 1.5 0.7109 1>
    Create Color Attribute=true
    Create Exponent Attribute=true
    Create Position Attribute=true
    Create Radius Attribute=true
    Enable Particle Light Enable=true
    Exponent=1
    Light Enabled=HlslExpression() {hlslExpression="Modulo(Particles.UniqueID, 8) == 0 ? 1: 0"}
    Position=&Add_Vector(
     A=$Output.FN_BoneFlow.PositionOnSpline
     B=&Multiply_VectorByFloat(Vector=&NormalizeVector(Vector To Normalize=$Output.FN_BoneFlow.RotatedBiNormal), Float=10)
    )
    Radius=14
    Scale Color=&FloatFromCurve(FloatCurve=DataInterface() {propertyValues {Curve {keys [
         <at 0 4 RCIM_Cubic>
         {arriveTangent=-3.8319, interpMode=RCIM_Cubic, leaveTangent=-3.8319, time=0.1693, value=1.2561}
         {arriveTangent=0.4201, interpMode=RCIM_Cubic, leaveTangent=0.4201, time=0.4182, value=2.3973}
         {arriveTangent=-4.1208, interpMode=RCIM_Cubic, leaveTangent=-4.1208, time=0.75, value=1.5}
         <at 1 0>
        ]}}})
   )
   &ScaleRibbonWidth(Ribbon Width Scale=&OneMinusFloat(Float=&MakeFloatFromInt(Int=&self:PartHiddenPartitionCheck(
       is Hand Hidden:NiagaraBool=$User.bisHandHidden
       Selection:NiagaraInt32=$Particles.Partition.ID
       is Foot Hidden:NiagaraBool=$User.bisFootHidden
      )))) {module=ScaleRibbonWidth001}
   /SetVariables(Particles.VisibilityTag=&self:PartHiddenPartitionCheck(
     is Hand Hidden:NiagaraBool=$User.bisHandHidden
     Selection:NiagaraInt32=$Particles.Partition.ID
     is Foot Hidden:NiagaraBool=$User.bisFootHidden
    )) {isSetParameters=true, module=SetVariables_#2}
  ]
  rendererClasses [&NiagaraRibbonRendererProperties, &NiagaraLightRendererProperties]
  renderers [
   {rendererClass=&NiagaraRibbonRendererProperties}
   {rendererClass=&NiagaraLightRendererProperties, rendererIndex=1}
  ]
 }
 Decal {
  emitterUpdateScript [
   &EmitterState()
   /SetVariables(Emitter.Buildup:NiagaraFloat=&MakeCustomFloatFromBool(
     Bool=$User.IsFrontEnd
     True Float=&FloatFromCurve(
      FloatCurve=DataInterface() {propertyValues {Curve {keys [
          <at 0 0 RCIM_Cubic>
          <at 0.02 1 RCIM_Cubic>
          {arriveTangent=-4.6406, interpMode=RCIM_Cubic, leaveTangent=-4.6407, tangentMode=RCTM_User, time=0.45, value=0}
         ]}}}
      CurveIndex=$Emitter.Age
     )
     False Float=&FloatFromCurve(
      FloatCurve=DataInterface() {propertyValues {Curve {keys [
          <at 0 0 RCIM_Cubic>
          <at 0.02 1 RCIM_Cubic>
          {arriveTangent=-2.2491, interpMode=RCIM_Cubic, leaveTangent=-2.2492, tangentMode=RCTM_User, time=1, value=0}
         ]}}}
      CurveIndex=$Emitter.Age
     )
    )) {isSetParameters=true, module=SetVariables_#3}
   &SpawnBurst_Instantaneous(Spawn Count=21, Spawn Probability=HlslExpression() {hlslExpression="Emitter.LoopCount == 0"}, Use Spawn Probability=true)
  ]
  particleSpawnScript [
   &FN_SetSpawnGroup_v2(Number of Spawn Groups=21)
   /Niagara/Modules/Spawn/Initialization/InitializeParticle(
    Color=<rgba 0.9473 0.9013 0.671 1>
    Lifetime=&MakeCustomFloatFromBool(Bool=$User.IsFrontEnd, True Float=0.7, False Float=1.1)
    Module.Write Color=true
    Module.Write Lifetime=true
    Module.Write Mass=true
    Module.Write Material Random=true
    Module.Write Position=false
    Module.Write Scale=false
    Module.Write SpriteRotation=false
    Module.Write SpriteSize=true
    Position=&SimulationPosition()
    Sprite Size=&Vector2DFromFloat(Value=&MaskFloatBySpawnGroup(
      Masked Float=26
      Passthrough Float=&MaskFloatBySpawnGroup(Masked Float=26, Passthrough Float=34, Spawn Group Mask=20)
      Spawn Group Mask=19
     ))
   )
   &DynamicMaterialParameters(
    Index 0 Param 1=0.5
    Index 0 Param 2=1
    Index 0 Param 3=1
    Module.Param0WriteEnabled=false
    Module.Param1WriteEnabled=true
    Module.Param2WriteEnabled=true
    Module.Param3WriteEnabled=true
   )
   &CameraOffset(Camera Offset Amount=16)
  ]
  particleUpdateScript [
   &ParticleState()
   &SkeletalMeshLocation(
    Apply Sampled Orientation to Mesh Particles=true
    Bone / Socket Index=$Particles.SpawnGroup
    Bone Sampling Mode=::NewEnumerator4
    Inherit Velocity=false
    MaskSpawnGroup=false
    Sampled Position Offset=&MaskVectorBySpawnGroup(
     Masked Vector=<8 0 0>
     Passthrough Vector=&MaskVectorBySpawnGroup(
      Masked Vector=<1 5 0>
      Passthrough Vector=&MaskVectorBySpawnGroup(Masked Vector=<-1 -5 0>, Passthrough Vector=<0 0 0>, Spawn Group Mask=20)
      Spawn Group Mask=19
     )
     Spawn Group Mask=10
    )
    Sampled Position Offset Coordinate Space:ENiagaraMeshTransforms=::NewEnumerator3
    Skeletal Mesh=$User.SkeletalMesh_Decal
   )
   &DynamicMaterialParameters(
    Index 0 Param 1=&Lerp_Float(A=2, B=0.1, Alpha=$Emitter.Buildup)
    Module.Param0WriteEnabled=true
    Module.Param1WriteEnabled=false
    Module.Param2WriteEnabled=false
    Module.Param3WriteEnabled=false
   ) {module=DynamicMaterialParameters001}
   &ScaleColor(
    Scale Alpha=&FloatFromCurve(FloatCurve=DataInterface() {propertyValues {Curve {keys [<at 0 0 RCIM_Cubic>,<at 0.2 1 RCIM_Cubic>,<at 1 0 RCIM_Cubic>]}}})
    Scale RGB=&MakeVectorFromLinearColorRGB(Color=&ColorFromCurve(ColorCurve=DataInterface() {propertyValues {
        AlphaCurve {keys []}
        BlueCurve {keys [<at 0 1>, {arriveTangent=-0.1823, interpMode=RCIM_Cubic, leaveTangent=-0.1823, time=0.15, value=2.1354}, <at 1 0.8177 RCIM_Cubic>]}
        GreenCurve {keys [<at 0 1>, {arriveTangent=-0.0091, interpMode=RCIM_Cubic, leaveTangent=-0.0091, time=0.15, value=2.5}, <at 1 0.9909 RCIM_Cubic>]}
        RedCurve {keys [<at 0 1>,<at 0.15 2.4878 RCIM_Cubic>,<at 1 1 RCIM_Cubic>]}
       }}))
    ScaleA=true
    ScaleRGB=true
   )
   /SetVariables(Particles.VisibilityTag=&self:PartHiddenDecalSKMCheck(
     is Hand Hidden:NiagaraBool=$User.bisHandHidden
     is Head Hidden:NiagaraBool=$User.bisHeadHidden
     is Foot Hidden:NiagaraBool=$User.bisFootHidden
     Selection:NiagaraInt32=$Particles.SpawnGroup
    )) {isSetParameters=true, module=SetVariables_#4}
  ]
  rendererClasses [&NiagaraSpriteRendererProperties]
  renderers [{rendererClass=&NiagaraSpriteRendererProperties}]
 }
 Lightning_Sprites {
  emitterSpawnScript [
   /SetVariables(Emitter.RandomSeed:NiagaraInt32=&RandomInt(Randomness Mode:ENiagaraRandomnessMode=::NewEnumerator0, Random Seed:NiagaraInt32=0, Fixed Random Seed:NiagaraBool=false)) {isSetParameters=true, module=SetVariables_#1}
   &FN_PartitionAttributes(Number of Bone Sets=6, Particles Per Partition=6, Partitions=6)
  ]
  emitterUpdateScript [
   &EmitterState()
   &SpawnRate(SpawnRate=50, Use Spawn Probability=true)
  ]
  particleSpawnScript [
   &FN_Particle_Partition_v5(Pattern:ENiagaraPartitionWaveForm=::NewEnumerator1, Seed:NiagaraInt32=$Emitter.RandomSeed)
   &FN_Initialize_BoneFlow_Attributes_v2(
    Angle=HlslExpression() {hlslExpression="Particles.Partition.ID == 5 ? rand(Particles.UniqueID) * 6.24 : 3.14 /2"}
    Radius:NiagaraFloat=&Multiply_Float(
     A=HlslExpression() {hlslExpression="Particles.Partition.ID == 5 ? 6 : 1"}
     B=/Niagara/DynamicInputs/Math/LerpMultipleFloats(
      Float Count:ENiagara_MultipleLerpCount=::NewEnumerator0
      1:NiagaraFloat=0.1
      2:NiagaraFloat=2
      3:NiagaraFloat=0.1
      4:NiagaraFloat=?
      5:NiagaraFloat=?
      Alpha:NiagaraFloat=$Particles.Partition.NormalizedID
     )
    )
   )
   &InitializeParticle(
    Color=<rgba 0.7448 0.6468 0.353 1>
    Lifetime=&UniformRangedFloat(Minimum=0.2, Maximum=0.3)
    Mesh Scale=<3 3 3>
    Module.Write Color=true
    Module.Write Lifetime=true
    Module.Write Mass=false
    Module.Write Material Random=true
    Module.Write Position=false
    Module.Write Scale=true
    Module.Write SpriteRotation=true
    Module.Write SpriteSize=true
    Sprite Rotation=&NormalizedAngleToDegrees(NormalizedAngle=&UniformRangedFloat(Minimum=0, Maximum=1))
    Sprite Size=&Vector2DFromFloat(Value=&UniformRangedFloat(Minimum=14, Maximum=18))
   )
   /SetVariables(Particles.SubImageIndex=&UniformRangedFloat(Minimum=0, Maximum=3)) {isSetParameters=true, module=SetVariables_#5}
   &CameraOffset(Camera Offset Amount=&Lerp_Float(
     A=&RandomRangeFloat(Minimum=-4, Maximum=4)
     B=&RandomRangeFloat(Minimum=-4, Maximum=10)
     Alpha=HlslExpression() {hlslExpression="Particles.Partition.ID >  1 ? 1 : 0"}
    ))
  ]
  particleUpdateScript [
   &ParticleState()
   &DynamicMaterialParameters(
    Index 0 Param 1=&RandomRangeFloat(Minimum=0, Maximum=1)
    Index 0 Param 2=&RandomRangeFloat(Minimum=0, Maximum=0.1)
    Index 0 Param 3=&FloatFromCurve(FloatCurve=DataInterface() {propertyValues {Curve {keys [<at 0 1 RCIM_Cubic>,<at 0.2 0.025 RCIM_Cubic>,<at 1 1>]}}})
    Module.Param0WriteEnabled=true
    Module.Param1WriteEnabled=true
    Module.Param2WriteEnabled=true
    Module.Param3WriteEnabled=false
   )
   &ScaleColor(
    Scale Alpha=&FloatFromCurve(FloatCurve=DataInterface() {propertyValues {Curve {keys [<at 0.75 1 RCIM_Cubic>, {arriveTangent=-10.3917, interpMode=RCIM_Cubic, leaveTangent=-10.3917, tangentMode=RCTM_User, time=1, value=0}]}}})
    Scale RGB=&MakeVectorFromLinearColorRGB(Color=&ColorFromCurve(ColorCurve=DataInterface() {propertyValues {
        AlphaCurve {keys []}
        BlueCurve {keys [<at 0 1>, {arriveTangent=-0.1823, interpMode=RCIM_Cubic, leaveTangent=-0.1823, time=0.15, value=3.8438}, <at 1 0.8177 RCIM_Cubic>]}
        GreenCurve {keys [<at 0 1>, {arriveTangent=-0.0091, interpMode=RCIM_Cubic, leaveTangent=-0.0091, time=0.15, value=4.5}, <at 1 0.9909 RCIM_Cubic>]}
        RedCurve {keys [<at 0 1>,<at 0.15 4.4781 RCIM_Cubic>,<at 1 1 RCIM_Cubic>]}
       }}))
    ScaleA=true
    ScaleRGB=true
   )
   &FN_BoneFlow(
    Angle:NiagaraFloat=$Particles.FN_Initialize_BoneFlow_Attributes_v2.Angle
    Height:NiagaraFloat=$Particles.Partition.NormalizedID
    Interpolate Mesh Sampling=false
    Radius:NiagaraFloat=$Particles.FN_Initialize_BoneFlow_Attributes_v2.Radius
    SkeletalMesh:NiagaraDataInterfaceSkeletalMesh=$User.SkeletalMesh
    Write Position Attribute=true
   )
   /SetVariables(Particles.VisibilityTag=&self:PartHiddenPartitionCheck(
     is Hand Hidden:NiagaraBool=$User.bisHandHidden
     Selection:NiagaraInt32=$Particles.Partition.ID
     is Foot Hidden:NiagaraBool=$User.bisFootHidden
    )) {isSetParameters=true, module=SetVariables_#2}
  ]
  rendererClasses [&NiagaraSpriteRendererProperties]
  renderers [{rendererClass=&NiagaraSpriteRendererProperties}]
 }
 Lightning_Sprites_Head {
  emitterUpdateScript [
   &EmitterState()
   &SpawnRate(SpawnRate=20, Use Spawn Probability=true)
  ]
  particleSpawnScript [
   &FN_Particle_Partition_v5(Pattern:ENiagaraPartitionWaveForm=::NewEnumerator1, Seed:NiagaraInt32=$Emitter.RandomSeed, Shuffle Partition ID=true)
   &InitializeParticle(
    Color=<rgba 0.7448 0.6468 0.353 1>
    Lifetime=&UniformRangedFloat(Minimum=0.15, Maximum=0.3)
    Mesh Scale=<3 3 3>
    Module.Write Color=true
    Module.Write Lifetime=true
    Module.Write Mass=false
    Module.Write Material Random=true
    Module.Write Position=false
    Module.Write Scale=true
    Module.Write SpriteRotation=true
    Module.Write SpriteSize=true
    Sprite Rotation=&NormalizedAngleToDegrees(NormalizedAngle=&UniformRangedFloat(Minimum=0, Maximum=1))
    Sprite Size=&Vector2DFromFloat(Value=&UniformRangedFloat(Minimum=12, Maximum=16))
   )
   /SetVariables(Particles.SubImageIndex=&UniformRangedFloat(Minimum=0, Maximum=3)) {isSetParameters=true, module=SetVariables_#5}
   &CameraOffset(Camera Offset Amount=&RandomRangeFloat(Minimum=-4, Maximum=6))
  ]
  particleUpdateScript [
   &ParticleState()
   &DynamicMaterialParameters(
    Index 0 Param 1=&RandomRangeFloat(Minimum=0, Maximum=1)
    Index 0 Param 2=&RandomRangeFloat(Minimum=0, Maximum=0.1)
    Index 0 Param 3=&FloatFromCurve(FloatCurve=DataInterface() {propertyValues {Curve {keys [<at 0 1 RCIM_Cubic>,<at 0.2 0.025 RCIM_Cubic>,<at 1 1>]}}})
    Module.Param0WriteEnabled=true
    Module.Param1WriteEnabled=true
    Module.Param2WriteEnabled=true
    Module.Param3WriteEnabled=false
   )
   &ScaleColor(
    Scale Alpha=&FloatFromCurve(FloatCurve=DataInterface() {propertyValues {Curve {keys [<at 0.75 1 RCIM_Cubic>, {arriveTangent=-10.3917, interpMode=RCIM_Cubic, leaveTangent=-10.3917, tangentMode=RCTM_User, time=1, value=0}]}}})
    Scale RGB=&Multiply_VectorByFloat(
     Vector=&MakeVectorFromLinearColorRGB(Color=&ColorFromCurve(ColorCurve=DataInterface() {propertyValues {
         AlphaCurve {keys []}
         BlueCurve {keys [<at 0 1>, {arriveTangent=-0.1823, interpMode=RCIM_Cubic, leaveTangent=-0.1823, time=0.15, value=3.8438}, <at 1 0.8177 RCIM_Cubic>]}
         GreenCurve {keys [<at 0 1>, {arriveTangent=-0.0091, interpMode=RCIM_Cubic, leaveTangent=-0.0091, time=0.15, value=4.5}, <at 1 0.9909 RCIM_Cubic>]}
         RedCurve {keys [<at 0 1>,<at 0.15 4.4781 RCIM_Cubic>,<at 1 1 RCIM_Cubic>]}
        }}))
     Float=1
    )
    ScaleA=true
    ScaleRGB=true
   )
   &SkeletalMeshLocation(
    Bone / Socket Index=18
    Bone Sampling Mode=::NewEnumerator4
    Inherit Velocity=false
    MaskSpawnGroup=false
    Sampled Position Offset=&RandomRangeVector(Minimum=<12 -4 -2>, Maximum=<20 4 2>)
    Sampled Position Offset Coordinate Space:ENiagaraMeshTransforms=::NewEnumerator3
    Skeletal Mesh=$User.SkeletalMesh
   )
   /SetVariables(Particles.VisibilityTag=&MakeCustomIntFromBool(Bool=$User.bisHeadHidden, True INT=1, False INT=0)) {isSetParameters=true}
  ]
  rendererClasses [&NiagaraSpriteRendererProperties]
  renderers [{rendererClass=&NiagaraSpriteRendererProperties}]
 }
 Sparks {
  emitterUpdateScript [
   &EmitterState()
   /SetVariables(Emitter.Buildup:NiagaraFloat=&FloatFromCurve(
     FloatCurve=DataInterface() {propertyValues {Curve {keys [
         <at 0 0 RCIM_Cubic>
         <at 0.1 0.9989 RCIM_Cubic>
         {arriveTangent=-0.0154, interpMode=RCIM_Cubic, leaveTangent=-0.0154, tangentMode=RCTM_User, time=0.5, value=0}
        ]}}}
     CurveIndex=$Emitter.Age
    )) {isSetParameters=true, module=SetVariables_#3}
   &SpawnRate(Spawn Probability=$Emitter.Buildup, SpawnRate=60, Use Spawn Probability=true)
  ]
  particleSpawnScript [
   &FN_SetSpawnGroup_v2(Number of Spawn Groups=19)
   &InitializeParticle(
    Color=<rgba 0.7448 0.7448 0.353 1>
    Lifetime=&UniformRangedFloat(Minimum=0.4, Maximum=0.5)
    Module.Write Color=true
    Module.Write Lifetime=true
    Module.Write Mass=true
    Module.Write Material Random=true
    Module.Write Position=false
    Module.Write Scale=false
    Module.Write SpriteRotation=false
    Module.Write SpriteSize=true
    Position=&SimulationPosition()
    Sprite Size=&UniformRangedVector2D(Minimum=<0.35 1>, Maximum=<0.75 2>)
   )
   &CameraOffset(Camera Offset Amount=12)
   &SkeletalMeshLocation(
    Apply Sampled Orientation to Mesh Particles=true
    Bone / Socket Index=$Particles.SpawnGroup
    Bone Sampling Mode=::NewEnumerator4
    Inherit Velocity=false
    MaskSpawnGroup=false
    Skeletal Mesh=$User.SkeletalMesh_Decal
   )
   /Niagara/Modules/Spawn/Location/SphereLocation(
    Hemisphere X=true
    MaskSpawnGroup=false
    Module.Override Randomness=false
    Sphere Origin:Vector3f=$Output.SkeletalMeshLocation.SampledPosition
    Sphere Radius=2
    Surface Only=false
   )
   &AddVelocity(Velocity=&UniformRangedVector(Minimum=<-125 -125 100>, Maximum=<125 125 200>))
  ]
  particleUpdateScript [
   &ParticleState()
   &AccelerationForce(Acceleration=<0 0 -980>)
   &SolveForcesAndVelocity(Clamp Velocity=false, Limit Acceleration=false)
   &ScaleSpriteSizeBySpeed(Max Scale Factor=<1 3>, Min Scale Factor=<1 2>)
   &ScaleSpriteSize(Scale Factor=&Vector2DFromFloat(Value=&FloatFromCurve(FloatCurve=DataInterface() {propertyValues {Curve {keys [{arriveTangent=-6.725, interpMode=RCIM_Cubic, leaveTangent=-6.725, tangentMode=RCTM_User, time=0.65, value=1}, <at 1 0 RCIM_Cubic>]}}})))
   &ScaleColor(
    Scale Alpha=&FloatFromCurve(FloatCurve=DataInterface() {propertyValues {Curve {keys [<at 0 0 RCIM_Cubic>,<at 0.2 1 RCIM_Cubic>,<at 1 0 RCIM_Cubic>]}}})
    Scale RGB=&MakeVectorFromLinearColorRGB(Color=&ColorFromCurve(ColorCurve=DataInterface() {propertyValues {
        AlphaCurve {keys []}
        BlueCurve {keys [<at 0 1>, {arriveTangent=-0.1823, interpMode=RCIM_Cubic, leaveTangent=-0.1823, time=0.15, value=3.8438}, <at 1 0.8177 RCIM_Cubic>]}
        GreenCurve {keys [<at 0 1>, {arriveTangent=-0.0091, interpMode=RCIM_Cubic, leaveTangent=-0.0091, time=0.15, value=4.5}, <at 1 0.9909 RCIM_Cubic>]}
        RedCurve {keys [<at 0 1>,<at 0.15 4.4781 RCIM_Cubic>,<at 1 1 RCIM_Cubic>]}
       }}))
    ScaleA=true
    ScaleRGB=true
   )
   /SetVariables(Particles.VisibilityTag=&self:PartHiddenDecalSKMCheck(
     is Hand Hidden:NiagaraBool=$User.bisHandHidden
     is Head Hidden:NiagaraBool=$User.bisHeadHidden
     is Foot Hidden:NiagaraBool=$User.bisFootHidden
     Selection:NiagaraInt32=$Particles.SpawnGroup
    )) {isSetParameters=true, module=SetVariables_#4}
  ]
  rendererClasses [&NiagaraSpriteRendererProperties]
  renderers [{rendererClass=&NiagaraSpriteRendererProperties}]
 }
}
summary {
 systemName=NS_GrilledCheese_Burst
 emitters [
  {emitterName=Decal, rendererClasses [&NiagaraSpriteRendererProperties]}
  {
   emitterName=BoneFlow_ArcEnergy
   rendererClasses [&NiagaraRibbonRendererProperties, &NiagaraLightRendererProperties]
  }
  {emitterName=Lightning_Sprites, rendererClasses [&NiagaraSpriteRendererProperties]}
  {emitterName=Lightning_Sprites_Head, rendererClasses [&NiagaraSpriteRendererProperties]}
  {emitterName=Sparks, rendererClasses [&NiagaraSpriteRendererProperties]}
 ]
 userVariables [
  User.bisFootHidden {type=NiagaraBool}
  User.bisHandHidden {type=NiagaraBool}
  User.bisHeadHidden {type=NiagaraBool}
  User.FinalStage {type=NiagaraFloat, defaultValue=0, description=""}
  User.IsFrontEnd {type=NiagaraBool}
  User.RibbonIDOffset {type=NiagaraInt32}
  User.SkeletalMesh {
   type=NiagaraDataInterfaceSkeletalMesh
   defaultValue {
    struct=NiagaraExt_VariableValue_DataInterface
    value {dataInterface=&self:NiagaraDataInterfaceSkeletalMesh_0}
   }
  }
  User.SkeletalMesh_Decal {
   type=NiagaraDataInterfaceSkeletalMesh
   defaultValue {
    struct=NiagaraExt_VariableValue_DataInterface
    value {dataInterface=&self:NiagaraDataInterfaceSkeletalMesh_1}
   }
  }
 ]
}
